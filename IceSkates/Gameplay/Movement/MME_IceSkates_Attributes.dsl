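; Movement mode extension for the Ice Skates item that records which item activated it, with debug prints referring to Ice Skates sprint.
; On activation it gets the owner pawn's current weapon and, if that is B_IceSkates, sets it as the activating item on the context's ItemActivated runtime data.
; Any other weapon triggers a shipping log warning instead.
; Ending the extension only prints a debug message.
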(event EventOnExtensionActivate
  (params (RuntimeContext :type MMERuntime%20Context%20Structure%20%28by%20ref%29))
  (bind _ownerpawn (Variables|FortMovementMode_BaseExtLogic|GetOwnerPawn))
  (bind _currentweapon (Variables|/Script/FortniteGame.FortPawn|GetCurrentWeapon _ownerpawn))
  (Development|PrintString 0 "Enable Ice Skates Sprint" true true "(R=0.125954,G=0.000000,B=1.000000,A=1.000000)")
  (bind _returnvalue (Class|FortMovementModeExtBlueprintLibrary|GetRuntimeDatafromContext RuntimeContext "/Script/FortniteGame.FortMovementMode_ItemActivatedRuntimeData"))
  (if (Utilities|IsA _currentweapon "/IceSkates/Gameplay/B_IceSkates.B_IceSkates_C")
    (Class|FortMovementModeItemActivatedRuntimeData|SetActivatingItemfromWeapon _returnvalue _currentweapon)
    (else
      (Fortnite|FortShippingLog _ownerpawn "MME_IceSkates_Attributes  activated but the CurrentWeapon is not  B_IceSkates! Cannot set activating item from weapon reliably." true))))

(event EventOnExtensionEnd
  (params (RuntimeContext :type MMERuntime%20Context%20Structure%20%28by%20ref%29))
  (Development|PrintString 0 "Disable Ice Skates Sprint" true true "(R=0.125954,G=0.000000,B=1.000000,A=1.000000)"))
